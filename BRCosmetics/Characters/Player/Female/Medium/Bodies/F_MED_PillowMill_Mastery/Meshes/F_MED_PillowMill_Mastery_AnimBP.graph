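# Anim Blueprint for the F_MED_PillowMill_Mastery cosmetic body, adding deformation and secondary motion to laces and a gun on the character.
# It copies the attached parent's pose into a deform Control Rig, passing a flag set each update when LOD is 2 or higher.
# Bone constraints then align lace bones and blend them with foot and calf bones.
# AnimDynamics springs jiggle the laces, then the gun after its thigh-weighted constraint.
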AnimBlueprint /BRCosmetics/Characters/Player/Female/Medium/Bodies/F_MED_PillowMill_Mastery/Meshes/F_MED_PillowMill_Mastery_AnimBP
target_skeleton=&/F_MED_PillowMill_Mastery_Skeleton
graphs [
 AnimGraph {nodes [
   n1 = Control Rig!(Source=Copy Pose From Mesh() {bUseAttachedParent=true}, Alpha, IsGreated lod 1:bool=Get LOD is Greater 1() -> (LOD is Greater 1:bool)) -> (Pose) {
    alphaBoolBlend {blendOption=Linear}
    controlRigAssetReference {blueprintRigClass=&/F_MED_PillowMill_Mastery_Deform_CtrlRig.F_MED_PillowMill_Mastery_Deform_CtrlRig_C}
   }
   n2 = "Constraint - dyn_laces_upperouter_l (R)"(ComponentPose=Local To Component(LocalPose=@n1), Alpha=1, ConstraintWeights_0=1) {
    alphaBoolBlend {blendOption=Linear}
    boneToModify {boneName=dyn_laces_upperouter_l}
    constraintSetup [{transformType=Rotation, targetBone {boneName=dyn_laces_bottominner_l}}]
    constraintWeights [1]
   }
   n3 = "Constraint - dyn_laces_upperinner_l (R)"(ComponentPose=@n2, Alpha=1, ConstraintWeights_0=1) {
    alphaBoolBlend {blendOption=Linear}
    boneToModify {boneName=dyn_laces_upperinner_l}
    constraintSetup [{transformType=Rotation, targetBone {boneName=dyn_laces_bottomouter_l}}]
    constraintWeights [1]
   }
   n4 = "Constraint - dyn_races_upperinner_r (R)"(ComponentPose=@n3, Alpha=1, ConstraintWeights_0=1) {
    alphaBoolBlend {blendOption=Linear}
    boneToModify {boneName=dyn_races_upperinner_r}
    constraintSetup [{transformType=Rotation, targetBone {boneName=dyn_races_bottominner_r}}]
    constraintWeights [1]
   }
   n5 = "Constraint - dyn_races_upperouter_r (R)"(ComponentPose=@n4, Alpha=1, ConstraintWeights_0=1) {
    alphaBoolBlend {blendOption=Linear}
    boneToModify {boneName=dyn_races_upperouter_r}
    constraintSetup [{transformType=Rotation, targetBone {boneName=dyn_races_bottomouter_r}}]
    constraintWeights [1]
   }
   n6 = "Constraint - deform_laces_bottomouter_l (PPPPP)"!(
    ComponentPose=@n5
    Alpha=0.5
    ConstraintWeights_0=0.9008
    ConstraintWeights_1:float=0
    ConstraintWeights_2:float=0
    ConstraintWeights_3:float=0.0539
    ConstraintWeights_4:float=0.0453
   ) -> (Pose) {
    alphaBoolBlend {blendOption=Linear}
    boneToModify {boneName=deform_laces_bottomouter_l}
    constraintSetup [
     {targetBone {boneName=foot_l}}
     {targetBone {boneName=deform_laces_m_l}}
     {targetBone {boneName=dyn_laces_upperinner_l}}
     {targetBone {boneName=deform_laces_bottomouter_l}}
     {targetBone {boneName=calf_twist_01_l}}
    ]
    constraintWeights [1,1,1,1,1]
   }
   n7 = "Constraint - deform_laces_bottominner_l (PPPP)"!(
    ComponentPose=@n6
    Alpha=1
    ConstraintWeights_0=0.9957
    ConstraintWeights_1:float=0.0017
    ConstraintWeights_2:float=0.0006
    ConstraintWeights_3:float=0.0021
   ) -> (Pose) {
    alphaBoolBlend {blendOption=Linear}
    boneToModify {boneName=deform_laces_bottominner_l}
    constraintSetup [
     {targetBone {boneName=foot_l}}
     {targetBone {boneName=ball_l}}
     {targetBone {boneName=deform_laces_bottominner_l}}
     {targetBone {boneName=calf_twist_01_l}}
    ]
    constraintWeights [1,1,1,1]
   }
   n8 = "Constraint - deform_races_bottomouter_r (PPPP)"!(
    ComponentPose=@n7
    Alpha=0.5
    ConstraintWeights_0=0.9419
    ConstraintWeights_1:float=0
    ConstraintWeights_2:float=0.0349
    ConstraintWeights_3:float=0.0233
   ) -> (Pose) {
    alphaBoolBlend {blendOption=Linear}
    boneToModify {boneName=deform_races_bottomouter_r}
    constraintSetup [
     {targetBone {boneName=foot_r}}
     {targetBone {boneName=dyn_races_upperinner_r}}
     {targetBone {boneName=deform_races_bottomouter_r}}
     {targetBone {boneName=calf_twist_01_r}}
    ]
    constraintWeights [1,1,1,1]
   }
   n9 = "Constraint - deform_races_bottominner_r (PPPP)"!(
    ComponentPose=@n8
    Alpha=1
    ConstraintWeights_0=0.9957
    ConstraintWeights_1:float=0.0017
    ConstraintWeights_2:float=0.0006
    ConstraintWeights_3:float=0.0021
   ) -> (Pose) {
    alphaBoolBlend {blendOption=Linear}
    boneToModify {boneName=deform_races_bottominner_r}
    constraintSetup [
     {targetBone {boneName=foot_r}}
     {targetBone {boneName=ball_r}}
     {targetBone {boneName=deform_races_bottominner_r}}
     {targetBone {boneName=calf_twist_01_r}}
    ]
    constraintWeights [1,1,1,1]
   }
   n10 = Anim Dynamics - Bone: dyn_laces_upperouter_l(ComponentPose=@n9, Alpha=1, ExternalForce="0,0,0") {
    angularDampingOverride=0.9999
    angularSpringConstant=0.5
    bGravityOverrideInSimSpace=true
    bLinearSpring=true
    bOverrideAngularDamping=true
    bOverrideLinearDamping=true
    linearDampingOverride=0.9999
    linearSpringConstant=50
    simulationSpace=RootRelative
    alphaBoolBlend {blendOption=Linear}
    boundBone {boneName=dyn_laces_upperouter_l}
    componentAppliedLinearAccClamp=<3000 3000 3000>
    componentLinearAccScale=<0.15 0.15 0.15>
    componentLinearVelScale=<0.15 0.15 0.15>
    gravityOverride=<1 1 1>
    physicsBodyDefinitions [{
      sphereCollisionRadius=0
      boundBone {boneName=dyn_laces_upperouter_l}
      boxExtents=<6 2 5>
      constraintSetup {angularLimitsMax=<0 5 0>, angularLimitsMin=<0 -5 -30>, angularTarget=<0 0 0>, linearAxesMax=<0 0 2>}
      localJointOffset=<1 0 0>
     }]
    relativeSpaceBone {boneName=root}
   }
   n11 = Anim Dynamics - Bone: dyn_laces_upperinner_l(ComponentPose=@n10, Alpha=1, ExternalForce="0,0,0") {
    angularDampingOverride=0.9999
    angularSpringConstant=0.5
    bGravityOverrideInSimSpace=true
    bLinearSpring=true
    bOverrideAngularDamping=true
    bOverrideLinearDamping=true
    linearDampingOverride=0.9999
    linearSpringConstant=50
    simulationSpace=RootRelative
    alphaBoolBlend {blendOption=Linear}
    boundBone {boneName=dyn_laces_upperinner_l}
    componentAppliedLinearAccClamp=<3000 3000 3000>
    componentLinearAccScale=<0.15 0.15 0.15>
    componentLinearVelScale=<0.15 0.15 0.15>
    gravityOverride=<1 1 1>
    physicsBodyDefinitions [{
      sphereCollisionRadius=0
      boundBone {boneName=dyn_laces_upperinner_l}
      boxExtents=<6 2 5>
      constraintSetup {angularLimitsMax=<0 5 0>, angularLimitsMin=<0 -5 -30>, angularTarget=<0 0 0>, linearAxesMin=<0 0 -2>}
      localJointOffset=<1 0 0>
     }]
    relativeSpaceBone {boneName=root}
   }
   n12 = Anim Dynamics - Bone: dyn_laces_bottominner_l(ComponentPose=@n11, Alpha=1, ExternalForce="0,0,0") {
    angularDampingOverride=0.9999
    angularSpringConstant=0.5
    bGravityOverrideInSimSpace=true
    bLinearSpring=true
    bOverrideAngularDamping=true
    bOverrideLinearDamping=true
    linearDampingOverride=0.9999
    linearSpringConstant=150
    simulationSpace=RootRelative
    alphaBoolBlend {blendOption=Linear}
    alphaScaleBiasClamp {scale=0.6}
    boundBone {boneName=dyn_laces_bottominner_l}
    componentAppliedLinearAccClamp=<3000 3000 3000>
    componentLinearAccScale=<0.15 0.15 0.15>
    componentLinearVelScale=<0.15 0.15 0.15>
    gravityOverride=<1 1 1>
    physicsBodyDefinitions [{
      sphereCollisionRadius=0
      boundBone {boneName=dyn_laces_bottominner_l}
      boxExtents=<6 2 5>
      constraintSetup {angularLimitsMax=<0 30 0>, angularLimitsMin=<0 -30 -30>, angularTarget=<0 0 0>}
      localJointOffset=<1 0 0>
     }]
    relativeSpaceBone {boneName=root}
   }
   n13 = Anim Dynamics - Bone: dyn_laces_bottomouter_l(ComponentPose=@n12, Alpha=1, ExternalForce="0,0,0") {
    angularDampingOverride=0.9999
    angularSpringConstant=0.5
    bGravityOverrideInSimSpace=true
    bLinearSpring=true
    bOverrideAngularDamping=true
    bOverrideLinearDamping=true
    linearDampingOverride=0.9999
    linearSpringConstant=150
    simulationSpace=RootRelative
    alphaBoolBlend {blendOption=Linear}
    alphaScaleBiasClamp {scale=0.6}
    boundBone {boneName=dyn_laces_bottomouter_l}
    componentAppliedLinearAccClamp=<3000 3000 3000>
    componentLinearAccScale=<0.15 0.15 0.15>
    componentLinearVelScale=<0.15 0.15 0.15>
    gravityOverride=<1 1 1>
    physicsBodyDefinitions [{
      sphereCollisionRadius=0
      boundBone {boneName=dyn_laces_bottomouter_l}
      boxExtents=<6 2 5>
      constraintSetup {angularLimitsMax=<0 30 0>, angularLimitsMin=<0 -30 -30>, angularTarget=<0 0 0>}
      localJointOffset=<0.5 0 0>
     }]
    relativeSpaceBone {boneName=root}
   }
   n14 = Anim Dynamics - Bone: dyn_races_upperinner_r(ComponentPose=@n13, Alpha=1, ExternalForce="0,0,0") {
    angularDampingOverride=0.9999
    angularSpringConstant=0.5
    bGravityOverrideInSimSpace=true
    bLinearSpring=true
    bOverrideAngularDamping=true
    bOverrideLinearDamping=true
    linearDampingOverride=0.9999
    linearSpringConstant=50
    simulationSpace=RootRelative
    alphaBoolBlend {blendOption=Linear}
    boundBone {boneName=dyn_races_upperinner_r}
    componentAppliedLinearAccClamp=<3000 3000 3000>
    componentLinearAccScale=<0.15 0.15 0.15>
    componentLinearVelScale=<0.15 0.15 0.15>
    gravityOverride=<1 1 1>
    physicsBodyDefinitions [{
      sphereCollisionRadius=0
      boundBone {boneName=dyn_races_upperinner_r}
      boxExtents=<6 2 5>
      constraintSetup {angularLimitsMax=<0 5 0>, angularLimitsMin=<0 -5 -30>, angularTarget=<0 0 0>, linearAxesMin=<0 0 -2>}
      localJointOffset=<1 0 0>
     }]
    relativeSpaceBone {boneName=root}
   }
   n15 = Anim Dynamics - Bone: dyn_races_upperouter_r(ComponentPose=@n14, Alpha=1, ExternalForce="0,0,0") {
    angularDampingOverride=0.9999
    angularSpringConstant=0.5
    bGravityOverrideInSimSpace=true
    bLinearSpring=true
    bOverrideAngularDamping=true
    bOverrideLinearDamping=true
    linearDampingOverride=0.9999
    linearSpringConstant=50
    simulationSpace=RootRelative
    alphaBoolBlend {blendOption=Linear}
    boundBone {boneName=dyn_races_upperouter_r}
    componentAppliedLinearAccClamp=<3000 3000 3000>
    componentLinearAccScale=<0.15 0.15 0.15>
    componentLinearVelScale=<0.15 0.15 0.15>
    gravityOverride=<1 1 1>
    physicsBodyDefinitions [{
      sphereCollisionRadius=0
      boundBone {boneName=dyn_races_upperouter_r}
      boxExtents=<6 2 5>
      constraintSetup {angularLimitsMax=<0 0 0>, angularLimitsMin=<0 -10 -30>, angularTarget=<0 0 0>, linearAxesMax=<0 0 2>}
      localJointOffset=<1 0 0>
     }]
    relativeSpaceBone {boneName=root}
   }
   n16 = Anim Dynamics - Bone: dyn_races_bottominner_r(ComponentPose=@n15, Alpha=1, ExternalForce="0,0,0") {
    angularDampingOverride=0.9999
    angularSpringConstant=0.5
    bGravityOverrideInSimSpace=true
    bLinearSpring=true
    bOverrideAngularDamping=true
    bOverrideLinearDamping=true
    linearDampingOverride=0.9999
    linearSpringConstant=150
    simulationSpace=RootRelative
    alphaBoolBlend {blendOption=Linear}
    alphaScaleBiasClamp {scale=0.6}
    boundBone {boneName=dyn_races_bottominner_r}
    componentAppliedLinearAccClamp=<3000 3000 3000>
    componentLinearAccScale=<0.15 0.15 0.15>
    componentLinearVelScale=<0.15 0.15 0.15>
    gravityOverride=<1 1 1>
    physicsBodyDefinitions [{
      sphereCollisionRadius=0
      boundBone {boneName=dyn_races_bottominner_r}
      boxExtents=<6 2 5>
      constraintSetup {angularLimitsMax=<0 30 0>, angularLimitsMin=<0 -30 -30>, angularTarget=<0 0 0>}
      localJointOffset=<1 0 0>
     }]
    relativeSpaceBone {boneName=root}
   }
   n17 = Anim Dynamics - Bone: dyn_races_bottomouter_r(ComponentPose=@n16, Alpha=1, ExternalForce="0,0,0") {
    angularDampingOverride=0.9999
    angularSpringConstant=0.5
    bGravityOverrideInSimSpace=true
    bLinearSpring=true
    bOverrideAngularDamping=true
    bOverrideLinearDamping=true
    linearDampingOverride=0.9999
    linearSpringConstant=150
    simulationSpace=RootRelative
    alphaBoolBlend {blendOption=Linear}
    alphaScaleBiasClamp {scale=0.6}
    boundBone {boneName=dyn_races_bottomouter_r}
    componentAppliedLinearAccClamp=<3000 3000 3000>
    componentLinearAccScale=<0.15 0.15 0.15>
    componentLinearVelScale=<0.15 0.15 0.15>
    gravityOverride=<1 1 1>
    physicsBodyDefinitions [{
      sphereCollisionRadius=0
      boundBone {boneName=dyn_races_bottomouter_r}
      boxExtents=<6 2 5>
      constraintSetup {angularLimitsMax=<0 30 0>, angularLimitsMin=<0 -30 -30>, angularTarget=<0 0 0>}
      localJointOffset=<0.5 0 0>
     }]
    relativeSpaceBone {boneName=root}
   }
   n18 = "Constraint - deform_gun (PPPPP)"!(
    ComponentPose=@n17
    Alpha=1
    ConstraintWeights_0=0.0343
    ConstraintWeights_1:float=0.4594
    ConstraintWeights_2:float=0.0001
    ConstraintWeights_3:float=0.5059
    ConstraintWeights_4:float=0.0003
   ) -> (Pose) {
    alphaBoolBlend {blendOption=Linear}
    boneToModify {boneName=deform_gun}
    constraintSetup [
     {targetBone {boneName=pelvis}}
     {targetBone {boneName=thigh_r}}
     {targetBone {boneName=deform_knee_out_r}}
     {targetBone {boneName=thigh_twist_01_r}}
     {targetBone {boneName=deform_groin_r}}
    ]
    constraintWeights [1,1,1,1,1]
   }
   n19 = Anim Dynamics - Bone: deform_gun(ComponentPose=@n18, Alpha=1, ExternalForce="0,0,0") {
    angularDampingOverride=0.9999
    angularSpringConstant=0.5
    bGravityOverrideInSimSpace=true
    bLinearSpring=true
    bOverrideAngularDamping=true
    bOverrideLinearDamping=true
    linearDampingOverride=0.9999
    linearSpringConstant=150
    simulationSpace=RootRelative
    alphaBoolBlend {blendOption=Linear}
    alphaScaleBiasClamp {scale=0.3}
    boundBone {boneName=deform_gun}
    componentAppliedLinearAccClamp=<3000 3000 3000>
    componentLinearAccScale=<0.15 0.15 0.15>
    componentLinearVelScale=<0.15 0.15 0.15>
    gravityOverride=<1 1 1>
    physicsBodyDefinitions [{
      sphereCollisionRadius=0
      boundBone {boneName=deform_gun}
      boxExtents=<6 10 6>
      constraintSetup {angularLimitsMax=<0 0 0>, angularLimitsMin=<0 0 0>, angularTarget=<0 0 0>, linearAxesMax=<1 0 1>, linearAxesMin=<0 0 -1>}
      localJointOffset=<0.5 0 0>
     }]
    relativeSpaceBone {boneName=root}
   }
   n20 = Anim Dynamics - Bone: dyn_gun(ComponentPose=@n19, Alpha=1, ExternalForce="0,0,0") {
    angularDampingOverride=0.9999
    angularSpringConstant=0.5
    bGravityOverrideInSimSpace=true
    bOverrideAngularDamping=true
    bOverrideLinearDamping=true
    linearDampingOverride=0.9999
    simulationSpace=RootRelative
    alphaBoolBlend {blendOption=Linear}
    alphaScaleBiasClamp {scale=0.3}
    boundBone {boneName=dyn_gun}
    componentAppliedLinearAccClamp=<3000 3000 3000>
    componentLinearAccScale=<0.15 0.15 0.15>
    componentLinearVelScale=<0.15 0.15 0.15>
    gravityOverride=<1 1 1>
    physicsBodyDefinitions [{
      sphereCollisionRadius=0
      boundBone {boneName=dyn_gun}
      boxExtents=<6 10 6>
      constraintSetup {angularLimitsMax=<0 0 0>, angularLimitsMin=<0 0 0>, angularTarget=<0 0 0>, linearAxesMax=<0 0 1>, linearAxesMin=<0 0 -1>}
      localJointOffset=<1 0 0>
     }]
    relativeSpaceBone {boneName=root}
   }
   Output Pose(Result=Component To Local(ComponentPose=@n20)) {layerGroup=None}
  ]}
 EventGraph {nodes [
   Event BlueprintUpdateAnimation()
   > Set LOD is Greater 1(LOD is Greater 1:bool="integer >= integer"(A=Get CurrentLOD(), B=2)) -> (Output_Get:bool)
  ]}
]
variables [LOD is Greater 1 {type=boolean}]
